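# Third-person Sound Cue for Milo's crouched footsteps on metal in the "below" elevation variant, built from downstairs step recordings.
# The first two quality branches mix a random pick of twelve walk steps with a louder random metal thump, each through a modulator.
# The other two play one modulated pick from six walk steps, whose random node is excluded from branch culling.
# Crouch attenuation and owner-limited, stop-oldest concurrency govern playback.
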SoundCue /Figment_Martin/Athena/Sounds/LTMs/Ashton/Milo/Footsteps/Athena_Milo_Crouch_Metal_Footstep_Below_3P_Cue
root=QualityLevel_0
nodes [
 {id=Mixer_0, children [Modulator_1,Modulator_3], properties {inputVolume [1,1.6]}}
 {id=Modulator_0, children [Random_5]}
 {id=Modulator_1, children [Random_2]}
 {id=Modulator_3, children [Random_1]}
 {id=QualityLevel_0, children [Mixer_0,Mixer_0,Modulator_0,Modulator_0]}
 {id=Random_1, children [WavePlayer_0,WavePlayer_1,WavePlayer_2,WavePlayer_3,WavePlayer_4,WavePlayer_5], properties {weights [1,1,1,1,1,1]}}
 {
  id=Random_2
  children [
   WavePlayer_7
   WavePlayer_10
   WavePlayer_21
   WavePlayer_22
   WavePlayer_23
   WavePlayer_24
   WavePlayer_25
   WavePlayer_26
   WavePlayer_27
   WavePlayer_28
   WavePlayer_29
   WavePlayer_30
  ]
  properties {weights [1,1,1,1,1,1,1,1,1,1,1,1]}
 }
 {
  id=Random_5
  children [WavePlayer_14, WavePlayer_15, WavePlayer_16, WavePlayer_17, WavePlayer_18, WavePlayer_19]
  properties {bShouldExcludeFromBranchCulling=true, weights [1,1,1,1,1,1]}
 }
 {
  id=WavePlayer_0
  properties {soundWaveAssetPtr=/Figment_Martin/Sounds/Fort_Step_Sounds/Soldier/Elevations/Downstairs/FS_Downstairs_Metal_Thump_10}
 }
 {
  id=WavePlayer_1
  properties {soundWaveAssetPtr=/Figment_Martin/Sounds/Fort_Step_Sounds/Soldier/Elevations/Downstairs/FS_Downstairs_Metal_Thump_11}
 }
 {
  id=WavePlayer_10
  properties {soundWaveAssetPtr=/Game/Sounds/Fort_Step_Sounds/Soldier/Elevations/Downstairs/Soldier_Step_Walk_Metal_02_Downstairs}
 }
 {
  id=WavePlayer_14
  properties {soundWaveAssetPtr=/Game/Sounds/Fort_Step_Sounds/Soldier/Elevations/Downstairs/Soldier_Step_Walk_Metal_01_Downstairs}
 }
 {
  id=WavePlayer_15
  properties {soundWaveAssetPtr=&Soldier_Step_Walk_Metal_02_Downstairs}
 }
 {
  id=WavePlayer_16
  properties {soundWaveAssetPtr=/Game/Sounds/Fort_Step_Sounds/Soldier/Elevations/Downstairs/Soldier_Step_Walk_Metal_03_Downstairs}
 }
 {
  id=WavePlayer_17
  properties {soundWaveAssetPtr=/Game/Sounds/Fort_Step_Sounds/Soldier/Elevations/Downstairs/Soldier_Step_Walk_Metal_04_Downstairs}
 }
 {
  id=WavePlayer_18
  properties {soundWaveAssetPtr=/Game/Sounds/Fort_Step_Sounds/Soldier/Elevations/Downstairs/Soldier_Step_Walk_Metal_05_Downstairs}
 }
 {
  id=WavePlayer_19
  properties {soundWaveAssetPtr=/Game/Sounds/Fort_Step_Sounds/Soldier/Elevations/Downstairs/Soldier_Step_Walk_Metal_06_Downstairs}
 }
 {
  id=WavePlayer_2
  properties {soundWaveAssetPtr=/Figment_Martin/Sounds/Fort_Step_Sounds/Soldier/Elevations/Downstairs/FS_Downstairs_Metal_Thump_12}
 }
 {
  id=WavePlayer_21
  properties {soundWaveAssetPtr=&Soldier_Step_Walk_Metal_03_Downstairs}
 }
 {
  id=WavePlayer_22
  properties {soundWaveAssetPtr=&Soldier_Step_Walk_Metal_04_Downstairs}
 }
 {
  id=WavePlayer_23
  properties {soundWaveAssetPtr=&Soldier_Step_Walk_Metal_05_Downstairs}
 }
 {
  id=WavePlayer_24
  properties {soundWaveAssetPtr=&Soldier_Step_Walk_Metal_06_Downstairs}
 }
 {
  id=WavePlayer_25
  properties {soundWaveAssetPtr=/Game/Sounds/Fort_Step_Sounds/Soldier/Elevations/Downstairs/Soldier_Step_Walk_Metal_07_Downstairs}
 }
 {
  id=WavePlayer_26
  properties {soundWaveAssetPtr=/Game/Sounds/Fort_Step_Sounds/Soldier/Elevations/Downstairs/Soldier_Step_Walk_Metal_08_Downstairs}
 }
 {
  id=WavePlayer_27
  properties {soundWaveAssetPtr=/Game/Sounds/Fort_Step_Sounds/Soldier/Elevations/Downstairs/Soldier_Step_Walk_Metal_09_Downstairs}
 }
 {
  id=WavePlayer_28
  properties {soundWaveAssetPtr=/Game/Sounds/Fort_Step_Sounds/Soldier/Elevations/Downstairs/Soldier_Step_Walk_Metal_10_Downstairs}
 }
 {
  id=WavePlayer_29
  properties {soundWaveAssetPtr=/Game/Sounds/Fort_Step_Sounds/Soldier/Elevations/Downstairs/Soldier_Step_Walk_Metal_11_Downstairs}
 }
 {
  id=WavePlayer_3
  properties {soundWaveAssetPtr=/Figment_Martin/Sounds/Fort_Step_Sounds/Soldier/Elevations/Downstairs/FS_Downstairs_Metal_Thump_108}
 }
 {
  id=WavePlayer_30
  properties {soundWaveAssetPtr=/Game/Sounds/Fort_Step_Sounds/Soldier/Elevations/Downstairs/Soldier_Step_Walk_Metal_12_Downstairs}
 }
 {id=WavePlayer_4, properties {soundWaveAssetPtr=/Game/Sounds/Fort_Step_Sounds/Soldier/Elevations/Downstairs/FS_Downstairs_Metal_Thump_08}}
 {
  id=WavePlayer_5
  properties {soundWaveAssetPtr=/Figment_Martin/Sounds/Fort_Step_Sounds/Soldier/Elevations/Downstairs/FS_Downstairs_Metal_Thump_09}
 }
 {
  id=WavePlayer_7
  properties {soundWaveAssetPtr=&Soldier_Step_Walk_Metal_01_Downstairs}
 }
]
properties {
 attenuationSettings=/Game/Sounds/Attenuation/Fort_Player_Footsteps_3P_Crouch_Athena_Att
 maxDistance=2048
 soundClassObject=/Game/Sounds/FortSoundClassesAndModes/Fort_Player_Footsteps_AboveOrBelow
 volumeMultiplier=0.6
 concurrencyOverrides {bLimitToOwner=true, maxCount=1, resolutionRule=StopOldest}
 concurrencySet [&Athena_Footstep_Local_Conc, &Athena_Footstep_Global_Conc]
}
