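# Sound Cue for Milo's crouch-sprint footsteps on dirt, using first-person player foley class and footstep attenuation.
# A Mixer layers a random generic Commando step with a random Milo movement sweetener, the latter passing through its own Modulator.
# A root Modulator raises the pitch of the mix and randomizes pitch and volume per play.
# The cue plays at reduced volume and references the local Athena footstep concurrency set.
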SoundCue /Figment_Martin/Athena/Sounds/LTMs/Ashton/Milo/Footsteps/Athena_Milo_Crouch_Sprint_Dirt_Footstep_Cue
root=Modulator_0
nodes [
 {id=Mixer_0, children [Random_1,Modulator_1], properties {inputVolume [1,1]}}
 {id=Modulator_0, children [Mixer_0], properties {pitchMax=1.5, pitchMin=1.3, volumeMax=1, volumeMin=0.7}}
 {id=Modulator_1, children [Random_4]}
 {
  id=Random_1
  children [WavePlayer_0,WavePlayer_1,WavePlayer_2,WavePlayer_3,WavePlayer_4,WavePlayer_5]
  properties {bShouldExcludeFromBranchCulling=true, weights [1,1,1,1,1,1]}
 }
 {
  id=Random_4
  children [WavePlayer_29, WavePlayer_30, WavePlayer_31, WavePlayer_32, WavePlayer_33, WavePlayer_34]
  properties {weights [1,1,1,1,1,1]}
 }
 {id=WavePlayer_0, properties {soundWaveAssetPtr=/Game/Sounds/Fortress_Foley/Commando/Commando_Step_001}}
 {id=WavePlayer_1, properties {soundWaveAssetPtr=/Game/Sounds/Fortress_Foley/Commando/Commando_Step_002}}
 {id=WavePlayer_2, properties {soundWaveAssetPtr=/Game/Sounds/Fortress_Foley/Commando/Commando_Step_003}}
 {id=WavePlayer_29, properties {soundWaveAssetPtr=/Figment_Martin/Athena/Sounds/LTMs/Ashton/Milo/Ashton_Milo_Movment_Sweetener_01}}
 {id=WavePlayer_3, properties {soundWaveAssetPtr=/Game/Sounds/Fortress_Foley/Commando/Commando_Step_004}}
 {id=WavePlayer_30, properties {soundWaveAssetPtr=/Figment_Martin/Athena/Sounds/LTMs/Ashton/Milo/Ashton_Milo_Movment_Sweetener_02}}
 {id=WavePlayer_31, properties {soundWaveAssetPtr=/Figment_Martin/Athena/Sounds/LTMs/Ashton/Milo/Ashton_Milo_Movment_Sweetener_03}}
 {id=WavePlayer_32, properties {soundWaveAssetPtr=/Figment_Martin/Athena/Sounds/LTMs/Ashton/Milo/Ashton_Milo_Movment_Sweetener_04}}
 {id=WavePlayer_33, properties {soundWaveAssetPtr=/Figment_Martin/Athena/Sounds/LTMs/Ashton/Milo/Ashton_Milo_Movment_Sweetener_05}}
 {id=WavePlayer_34, properties {soundWaveAssetPtr=/Figment_Martin/Athena/Sounds/LTMs/Ashton/Milo/Ashton_Milo_Movment_Sweetener_06}}
 {id=WavePlayer_4, properties {soundWaveAssetPtr=/Game/Sounds/Fortress_Foley/Commando/Commando_Step_006}}
 {id=WavePlayer_5, properties {soundWaveAssetPtr=/Game/Sounds/Fortress_Foley/Commando/Commando_Step_010}}
]
properties {
 attenuationSettings=/Game/Sounds/Attenuation/Fort_Player_Footsteps_1P_CrouchSprint_Athena_Att
 maxDistance=2097152
 soundClassObject=&Fort_Player_Foley_1P_Athena
 volumeMultiplier=0.4
 concurrencyOverrides {bLimitToOwner=true, maxCount=1, resolutionRule=StopOldest}
 concurrencySet [&Athena_Footstep_Local_Conc]
}
