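# Third-person cue for Milo crouch-sprint footsteps on metal heard from above, built from Soldier "Upstairs" metal walk samples.
# A quality-level node feeds a random pick of ten waves for the first two levels and a three-wave subset for the last two.
# It plays at reduced volume through the above-or-below footstep sound class and crouch-sprint attenuation, with local and global footstep concurrency sets.
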SoundCue /Figment_Martin/Athena/Sounds/LTMs/Ashton/Milo/Footsteps/Athena_Milo_Crouch_Sprint_Metal_Footstep_Above_3P_Cue
root=QualityLevel_1
nodes [
 {id=QualityLevel_1, children [Random_1,Random_1,Random_2,Random_2]}
 {
  id=Random_1
  children [
   WavePlayer_12
   WavePlayer_13
   WavePlayer_14
   WavePlayer_22
   WavePlayer_23
   WavePlayer_24
   WavePlayer_25
   WavePlayer_26
   WavePlayer_27
   WavePlayer_28
  ]
  properties {weights [1,1,1,1,1,1,1,1,1,1]}
 }
 {id=Random_2, children [WavePlayer_6,WavePlayer_7,WavePlayer_8], properties {weights [1,1,1]}}
 {
  id=WavePlayer_12
  properties {soundWaveAssetPtr=/Game/Sounds/Fort_Step_Sounds/Soldier/Elevations/Upstairs/Soldier_Step_Walk_Metal_04_Upstairs2}
 }
 {
  id=WavePlayer_13
  properties {soundWaveAssetPtr=/Game/Sounds/Fort_Step_Sounds/Soldier/Elevations/Upstairs/Soldier_Step_Walk_Metal_11_Upstairs2}
 }
 {
  id=WavePlayer_14
  properties {soundWaveAssetPtr=/Game/Sounds/Fort_Step_Sounds/Soldier/Elevations/Upstairs/Soldier_Step_Walk_Metal_01_Upstairs2}
 }
 {
  id=WavePlayer_22
  properties {soundWaveAssetPtr=/Game/Sounds/Fort_Step_Sounds/Soldier/Elevations/Upstairs/Soldier_Step_Walk_Metal_02_Upstairs2}
 }
 {
  id=WavePlayer_23
  properties {soundWaveAssetPtr=/Game/Sounds/Fort_Step_Sounds/Soldier/Elevations/Upstairs/Soldier_Step_Walk_Metal_03_Upstairs2}
 }
 {
  id=WavePlayer_24
  properties {soundWaveAssetPtr=/Game/Sounds/Fort_Step_Sounds/Soldier/Elevations/Upstairs/Soldier_Step_Walk_Metal_05_Upstairs2}
 }
 {
  id=WavePlayer_25
  properties {soundWaveAssetPtr=/Game/Sounds/Fort_Step_Sounds/Soldier/Elevations/Upstairs/Soldier_Step_Walk_Metal_06_Upstairs2}
 }
 {
  id=WavePlayer_26
  properties {soundWaveAssetPtr=/Game/Sounds/Fort_Step_Sounds/Soldier/Elevations/Upstairs/Soldier_Step_Walk_Metal_07_Upstairs2}
 }
 {
  id=WavePlayer_27
  properties {soundWaveAssetPtr=/Game/Sounds/Fort_Step_Sounds/Soldier/Elevations/Upstairs/Soldier_Step_Walk_Metal_10_Upstairs2}
 }
 {
  id=WavePlayer_28
  properties {soundWaveAssetPtr=/Game/Sounds/Fort_Step_Sounds/Soldier/Elevations/Upstairs/Soldier_Step_Walk_Metal_12_Upstairs2}
 }
 {
  id=WavePlayer_6
  properties {soundWaveAssetPtr=&Soldier_Step_Walk_Metal_04_Upstairs2}
 }
 {
  id=WavePlayer_7
  properties {soundWaveAssetPtr=&Soldier_Step_Walk_Metal_11_Upstairs2}
 }
 {
  id=WavePlayer_8
  properties {soundWaveAssetPtr=&Soldier_Step_Walk_Metal_01_Upstairs2}
 }
]
properties {
 attenuationSettings=/Game/Sounds/Attenuation/Fort_Player_Footsteps_3P_CrouchSprint_Athena_Att
 maxDistance=2560
 soundClassObject=/Game/Sounds/FortSoundClassesAndModes/Fort_Player_Footsteps_AboveOrBelow
 volumeMultiplier=0.5
 concurrencyOverrides {bLimitToOwner=true, maxCount=1, resolutionRule=StopOldest}
 concurrencySet [&Athena_Footstep_Local_Conc, &Athena_Footstep_Global_Conc]
}
