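# Sound cue for Milo's crouch-sprint footsteps on snow in the Ashton LTM, set up as first-person player foley.
# A Mixer layers two Modulator branches: a random pick of twelve shared Soldier snow walk steps, and a random pick of six Milo movement sweeteners.
# The cue plays at reduced volume through the first-person crouch-sprint footstep attenuation and the local footstep concurrency set.
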SoundCue /Figment_Martin/Athena/Sounds/LTMs/Ashton/Milo/Footsteps/Athena_Milo_Crouch_Sprint_Snow_Footstep_Cue
root=Mixer_0
nodes [
 {id=Mixer_0, children [Modulator_1,Modulator_0], properties {inputVolume [1,1]}}
 {id=Modulator_0, children [Random_4]}
 {id=Modulator_1, children [Random_2]}
 {
  id=Random_2
  children [
   WavePlayer_18
   WavePlayer_19
   WavePlayer_20
   WavePlayer_21
   WavePlayer_22
   WavePlayer_23
   WavePlayer_24
   WavePlayer_25
   WavePlayer_26
   WavePlayer_27
   WavePlayer_28
   WavePlayer_29
  ]
  properties {weights [1,1,1,1,1,1,1,1,1,1,1,1]}
 }
 {
  id=Random_4
  children [WavePlayer_0, WavePlayer_30, WavePlayer_31, WavePlayer_32, WavePlayer_33, WavePlayer_34]
  properties {weights [1,1,1,1,1,1]}
 }
 {id=WavePlayer_0, properties {soundWaveAssetPtr=/Figment_Martin/Athena/Sounds/LTMs/Ashton/Milo/Ashton_Milo_Movment_Sweetener_01}}
 {id=WavePlayer_18, properties {soundWaveAssetPtr=/Game/Sounds/Fort_Step_Sounds/Soldier/Soldier_Step_Walk_Snow_10}}
 {id=WavePlayer_19, properties {soundWaveAssetPtr=/Game/Sounds/Fort_Step_Sounds/Soldier/Soldier_Step_Walk_Snow_11}}
 {id=WavePlayer_20, properties {soundWaveAssetPtr=/Game/Sounds/Fort_Step_Sounds/Soldier/Soldier_Step_Walk_Snow_12}}
 {id=WavePlayer_21, properties {soundWaveAssetPtr=/Game/Sounds/Fort_Step_Sounds/Soldier/Soldier_Step_Walk_Snow_01}}
 {id=WavePlayer_22, properties {soundWaveAssetPtr=/Game/Sounds/Fort_Step_Sounds/Soldier/Soldier_Step_Walk_Snow_02}}
 {id=WavePlayer_23, properties {soundWaveAssetPtr=/Game/Sounds/Fort_Step_Sounds/Soldier/Soldier_Step_Walk_Snow_03}}
 {id=WavePlayer_24, properties {soundWaveAssetPtr=/Game/Sounds/Fort_Step_Sounds/Soldier/Soldier_Step_Walk_Snow_04}}
 {id=WavePlayer_25, properties {soundWaveAssetPtr=/Game/Sounds/Fort_Step_Sounds/Soldier/Soldier_Step_Walk_Snow_05}}
 {id=WavePlayer_26, properties {soundWaveAssetPtr=/Game/Sounds/Fort_Step_Sounds/Soldier/Soldier_Step_Walk_Snow_06}}
 {id=WavePlayer_27, properties {soundWaveAssetPtr=/Game/Sounds/Fort_Step_Sounds/Soldier/Soldier_Step_Walk_Snow_07}}
 {id=WavePlayer_28, properties {soundWaveAssetPtr=/Game/Sounds/Fort_Step_Sounds/Soldier/Soldier_Step_Walk_Snow_08}}
 {id=WavePlayer_29, properties {soundWaveAssetPtr=/Game/Sounds/Fort_Step_Sounds/Soldier/Soldier_Step_Walk_Snow_09}}
 {id=WavePlayer_30, properties {soundWaveAssetPtr=/Figment_Martin/Athena/Sounds/LTMs/Ashton/Milo/Ashton_Milo_Movment_Sweetener_02}}
 {id=WavePlayer_31, properties {soundWaveAssetPtr=/Figment_Martin/Athena/Sounds/LTMs/Ashton/Milo/Ashton_Milo_Movment_Sweetener_03}}
 {id=WavePlayer_32, properties {soundWaveAssetPtr=/Figment_Martin/Athena/Sounds/LTMs/Ashton/Milo/Ashton_Milo_Movment_Sweetener_04}}
 {id=WavePlayer_33, properties {soundWaveAssetPtr=/Figment_Martin/Athena/Sounds/LTMs/Ashton/Milo/Ashton_Milo_Movment_Sweetener_05}}
 {id=WavePlayer_34, properties {soundWaveAssetPtr=/Figment_Martin/Athena/Sounds/LTMs/Ashton/Milo/Ashton_Milo_Movment_Sweetener_06}}
]
properties {
 attenuationSettings=/Game/Sounds/Attenuation/Fort_Player_Footsteps_1P_CrouchSprint_Athena_Att
 maxDistance=2097152
 soundClassObject=&Fort_Player_Foley_1P_Athena
 volumeMultiplier=0.4
 concurrencyOverrides {bLimitToOwner=true, maxCount=1, resolutionRule=StopOldest}
 concurrencySet [&Athena_Footstep_Local_Conc]
}
