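# Sound cue for third-person crouched footsteps on stone heard from above, named for the Milo LTM, built from Soldier upstairs walk samples.
# A quality-level switch selects either a twelve-sample random pool or a three-sample subset for lower-cost tiers.
# A modulator applies a fixed volume reduction to the result.
# Playback uses crouch footstep attenuation, the above-or-below sound class, and local and global footstep concurrency sets.
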SoundCue /Figment_Martin/Athena/Sounds/LTMs/Ashton/Milo/Footsteps/Athena_Milo_Crouch_Stone_Footstep_Above_3P_Cue
root=Modulator_3
nodes [
 {id=Modulator_3, children [QualityLevel_1], properties {volumeMax=0.7, volumeMin=0.7}}
 {id=QualityLevel_1, children [Random_1,Random_1,Random_2,Random_2]}
 {
  id=Random_1
  children [
   WavePlayer_2
   WavePlayer_3
   WavePlayer_4
   WavePlayer_5
   WavePlayer_7
   WavePlayer_10
   WavePlayer_0
   WavePlayer_1
   WavePlayer_20
   WavePlayer_21
   WavePlayer_22
   WavePlayer_25
  ]
  properties {weights [1,1,1,1,1,1,1,1,1,1,1,1]}
 }
 {id=Random_2, children [WavePlayer_26, WavePlayer_27, WavePlayer_28], properties {weights [1,1,1]}}
 {
  id=WavePlayer_0
  properties {soundWaveAssetPtr=/Game/Sounds/Fort_Step_Sounds/Soldier/Elevations/Upstairs/Soldier_Step_Walk_Stone_07_Upstairs2}
 }
 {
  id=WavePlayer_1
  properties {soundWaveAssetPtr=/Game/Sounds/Fort_Step_Sounds/Soldier/Elevations/Upstairs/Soldier_Step_Walk_Stone_08_Upstairs2}
 }
 {
  id=WavePlayer_10
  properties {soundWaveAssetPtr=/Game/Sounds/Fort_Step_Sounds/Soldier/Elevations/Upstairs/Soldier_Step_Walk_Stone_06_Upstairs2}
 }
 {
  id=WavePlayer_2
  properties {soundWaveAssetPtr=/Game/Sounds/Fort_Step_Sounds/Soldier/Elevations/Upstairs/Soldier_Step_Walk_Stone_01_Upstairs2}
 }
 {
  id=WavePlayer_20
  properties {soundWaveAssetPtr=/Game/Sounds/Fort_Step_Sounds/Soldier/Elevations/Upstairs/Soldier_Step_Walk_Stone_09_Upstairs2}
 }
 {
  id=WavePlayer_21
  properties {soundWaveAssetPtr=/Game/Sounds/Fort_Step_Sounds/Soldier/Elevations/Upstairs/Soldier_Step_Walk_Stone_10_Upstairs2}
 }
 {
  id=WavePlayer_22
  properties {soundWaveAssetPtr=/Game/Sounds/Fort_Step_Sounds/Soldier/Elevations/Upstairs/Soldier_Step_Walk_Stone_11_Upstairs2}
 }
 {
  id=WavePlayer_25
  properties {soundWaveAssetPtr=/Game/Sounds/Fort_Step_Sounds/Soldier/Elevations/Upstairs/Soldier_Step_Walk_Stone_12_Upstairs2}
 }
 {
  id=WavePlayer_26
  properties {soundWaveAssetPtr=&Soldier_Step_Walk_Stone_01_Upstairs2}
 }
 {
  id=WavePlayer_27
  properties {soundWaveAssetPtr=/Game/Sounds/Fort_Step_Sounds/Soldier/Elevations/Upstairs/Soldier_Step_Walk_Stone_02_Upstairs2}
 }
 {
  id=WavePlayer_28
  properties {soundWaveAssetPtr=/Game/Sounds/Fort_Step_Sounds/Soldier/Elevations/Upstairs/Soldier_Step_Walk_Stone_03_Upstairs2}
 }
 {
  id=WavePlayer_3
  properties {soundWaveAssetPtr=&Soldier_Step_Walk_Stone_02_Upstairs2}
 }
 {
  id=WavePlayer_4
  properties {soundWaveAssetPtr=&Soldier_Step_Walk_Stone_03_Upstairs2}
 }
 {
  id=WavePlayer_5
  properties {soundWaveAssetPtr=/Game/Sounds/Fort_Step_Sounds/Soldier/Elevations/Upstairs/Soldier_Step_Walk_Stone_04_Upstairs2}
 }
 {
  id=WavePlayer_7
  properties {soundWaveAssetPtr=/Game/Sounds/Fort_Step_Sounds/Soldier/Elevations/Upstairs/Soldier_Step_Walk_Stone_05_Upstairs2}
 }
]
properties {
 attenuationSettings=/Game/Sounds/Attenuation/Fort_Player_Footsteps_3P_Crouch_Athena_Att
 maxDistance=2048
 soundClassObject=/Game/Sounds/FortSoundClassesAndModes/Fort_Player_Footsteps_AboveOrBelow
 volumeMultiplier=0.4
 concurrencyOverrides {bLimitToOwner=true, maxCount=1, resolutionRule=StopOldest}
 concurrencySet [&Athena_Footstep_Local_Conc, &Athena_Footstep_Global_Conc]
}
