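# Sound cue for the Ashton Milo character's jump on dirt, routed to the Athena first-person player foley class with player landing attenuation.
# A mixer layers a random Commando jump foley wave, pitched up with varied volume, over a random Milo movement sweetener through a default modulator.
# A concurrency override allows one instance per owner and blocks new plays while it sounds.
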SoundCue /Figment_Martin/Athena/Sounds/LTMs/Ashton/Milo/Footsteps/Athena_Milo_Jump_Dirt_Cue
root=Mixer_0
nodes [
 {id=Mixer_0, children [Modulator_2,Modulator_1], properties {inputVolume [1,1]}}
 {id=Modulator_1, children [Random_4]}
 {id=Modulator_2, children [Random_0], properties {pitchMax=1.3, pitchMin=1.1, volumeMax=1, volumeMin=0.7}}
 {id=Random_0, children [WavePlayer_349, WavePlayer_352, WavePlayer_353], properties {bShouldExcludeFromBranchCulling=true, weights [1,1,1]}}
 {
  id=Random_4
  children [WavePlayer_29, WavePlayer_30, WavePlayer_31, WavePlayer_32, WavePlayer_33, WavePlayer_34]
  properties {weights [1,1,1,1,1,1]}
 }
 {id=WavePlayer_29, properties {soundWaveAssetPtr=/Figment_Martin/Athena/Sounds/LTMs/Ashton/Milo/Ashton_Milo_Movment_Sweetener_01}}
 {id=WavePlayer_30, properties {soundWaveAssetPtr=/Figment_Martin/Athena/Sounds/LTMs/Ashton/Milo/Ashton_Milo_Movment_Sweetener_02}}
 {id=WavePlayer_31, properties {soundWaveAssetPtr=/Figment_Martin/Athena/Sounds/LTMs/Ashton/Milo/Ashton_Milo_Movment_Sweetener_03}}
 {id=WavePlayer_32, properties {soundWaveAssetPtr=/Figment_Martin/Athena/Sounds/LTMs/Ashton/Milo/Ashton_Milo_Movment_Sweetener_04}}
 {id=WavePlayer_33, properties {soundWaveAssetPtr=/Figment_Martin/Athena/Sounds/LTMs/Ashton/Milo/Ashton_Milo_Movment_Sweetener_05}}
 {id=WavePlayer_34, properties {soundWaveAssetPtr=/Figment_Martin/Athena/Sounds/LTMs/Ashton/Milo/Ashton_Milo_Movment_Sweetener_06}}
 {id=WavePlayer_349, properties {soundWaveAssetPtr=/Game/Sounds/Fortress_Foley/Commando/Commando_Jump_006}}
 {id=WavePlayer_352, properties {soundWaveAssetPtr=/Game/Sounds/Fortress_Foley/Commando/Commando_Jump_003}}
 {id=WavePlayer_353, properties {soundWaveAssetPtr=/Game/Sounds/Fortress_Foley/Commando/Commando_Jump_004}}
]
properties {
 attenuationSettings=/Game/Sounds/Attenuation/Fort_Player_Land_Athena_Att
 bOverrideConcurrency=true
 maxDistance=2097152
 soundClassObject=&Fort_Player_Foley_1P_Athena
 concurrencyOverrides {bLimitToOwner=true, maxCount=1, resolutionRule=PreventNew}
}
