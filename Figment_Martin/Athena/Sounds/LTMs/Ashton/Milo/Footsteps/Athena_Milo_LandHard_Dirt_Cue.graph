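# Sound cue for Milo's hard landing on dirt in the Ashton LTM, layering generic landing foley with character-specific movement sweeteners.
# Random soldier landings mix with half-volume random Commando landings, then pass through a wide pitch and volume modulator.
# A separately modulated random pick of six Milo sweeteners mixes on top.
# It uses the first-person player foley sound class, landing attenuation and footstep-landing concurrency.
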SoundCue /Figment_Martin/Athena/Sounds/LTMs/Ashton/Milo/Footsteps/Athena_Milo_LandHard_Dirt_Cue
root=Mixer_0
nodes [
 {id=Mixer_0, children [Modulator_4,Modulator_0], properties {inputVolume [1,1]}}
 {id=Mixer_2, children [Random_3,Random_4], properties {inputVolume [1,0.5]}}
 {id=Modulator_0, children [Random_0]}
 {id=Modulator_4, children [Mixer_2], properties {pitchMax=1.1, pitchMin=0.9, volumeMax=1.3, volumeMin=0.7}}
 {
  id=Random_0
  children [WavePlayer_29, WavePlayer_30, WavePlayer_31, WavePlayer_32, WavePlayer_33, WavePlayer_34]
  properties {weights [1,1,1,1,1,1]}
 }
 {id=Random_3, children [WavePlayer_19, WavePlayer_20], properties {bShouldExcludeFromBranchCulling=true, weights [1,1]}}
 {id=Random_4, children [WavePlayer_3,WavePlayer_4,WavePlayer_5], properties {bShouldExcludeFromBranchCulling=true, weights [1,1,1]}}
 {id=WavePlayer_19, properties {soundWaveAssetPtr=/Game/Sounds/Fort_Step_Sounds/Human/soldier_land_01}}
 {id=WavePlayer_20, properties {soundWaveAssetPtr=/Game/Sounds/Fort_Step_Sounds/Human/soldier_land_02}}
 {id=WavePlayer_29, properties {soundWaveAssetPtr=/Figment_Martin/Athena/Sounds/LTMs/Ashton/Milo/Ashton_Milo_Movment_Sweetener_01}}
 {id=WavePlayer_3, properties {soundWaveAssetPtr=/Game/Sounds/Fortress_Foley/Commando/Commando_Land_006}}
 {id=WavePlayer_30, properties {soundWaveAssetPtr=/Figment_Martin/Athena/Sounds/LTMs/Ashton/Milo/Ashton_Milo_Movment_Sweetener_02}}
 {id=WavePlayer_31, properties {soundWaveAssetPtr=/Figment_Martin/Athena/Sounds/LTMs/Ashton/Milo/Ashton_Milo_Movment_Sweetener_03}}
 {id=WavePlayer_32, properties {soundWaveAssetPtr=/Figment_Martin/Athena/Sounds/LTMs/Ashton/Milo/Ashton_Milo_Movment_Sweetener_04}}
 {id=WavePlayer_33, properties {soundWaveAssetPtr=/Figment_Martin/Athena/Sounds/LTMs/Ashton/Milo/Ashton_Milo_Movment_Sweetener_05}}
 {id=WavePlayer_34, properties {soundWaveAssetPtr=/Figment_Martin/Athena/Sounds/LTMs/Ashton/Milo/Ashton_Milo_Movment_Sweetener_06}}
 {id=WavePlayer_4, properties {soundWaveAssetPtr=/Game/Sounds/Fortress_Foley/Commando/Commando_Land_002}}
 {id=WavePlayer_5, properties {soundWaveAssetPtr=/Game/Sounds/Fortress_Foley/Commando/Commando_Land_005}}
]
properties {
 attenuationSettings=/Game/Sounds/Attenuation/Fort_Player_Land_Athena_Att
 maxDistance=2097152
 soundClassObject=&Fort_Player_Foley_1P_Athena
 volumeMultiplier=1.2
 concurrencyOverrides {maxCount=4, resolutionRule=PreventNew}
 concurrencySet [/Game/Sounds/Concurrency_Rules/Athena_Footstep_Land_Conc]
}
