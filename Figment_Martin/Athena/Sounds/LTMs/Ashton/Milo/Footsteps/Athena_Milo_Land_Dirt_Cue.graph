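# Landing sound cue for the Milo character on dirt in an Athena LTM, routed to the first-person player foley sound class.
# A Mixer layers quieter stereo Commando landing foley, Commando land sounds and Milo movement sweeteners.
# Each layer picks randomly from three waves through its own pitch and volume Modulator, with the foley pitched up.
# Shared landing attenuation and footstep-land concurrency govern playback.
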SoundCue /Figment_Martin/Athena/Sounds/LTMs/Ashton/Milo/Footsteps/Athena_Milo_Land_Dirt_Cue
root=Mixer_5
nodes [
 {id=Mixer_5, children [Modulator_6,Modulator_5,Modulator_0], properties {inputVolume [0.4,1,1]}}
 {id=Modulator_0, children [Random_2]}
 {id=Modulator_5, children [Random_0], properties {pitchMax=1.1, pitchMin=0.9, volumeMax=1.3, volumeMin=0.7}}
 {id=Modulator_6, children [Random_1], properties {pitchMax=1.4, pitchMin=1.1, volumeMax=1, volumeMin=0.5}}
 {id=Random_0, children [WavePlayer_0,WavePlayer_1,WavePlayer_2], properties {bShouldExcludeFromBranchCulling=true, weights [1,1,1]}}
 {id=Random_1, children [WavePlayer_338, WavePlayer_339, WavePlayer_341], properties {bShouldExcludeFromBranchCulling=true, weights [1,1,1]}}
 {id=Random_2, children [WavePlayer_3,WavePlayer_4,WavePlayer_5], properties {weights [1,1,1]}}
 {id=WavePlayer_0, properties {soundWaveAssetPtr=/Game/Sounds/Fortress_Foley/Commando/Commando_Land_006}}
 {id=WavePlayer_1, properties {soundWaveAssetPtr=/Game/Sounds/Fortress_Foley/Commando/Commando_Land_002}}
 {id=WavePlayer_2, properties {soundWaveAssetPtr=/Game/Sounds/Fortress_Foley/Commando/Commando_Land_005}}
 {id=WavePlayer_3, properties {soundWaveAssetPtr=/Figment_Martin/Athena/Sounds/LTMs/Ashton/Milo/Ashton_Milo_Movment_Sweetener_01}}
 {id=WavePlayer_338, properties {soundWaveAssetPtr=/Game/Sounds/Fortress_Foley/Commando/Commando_Land_Foley_Stereo_003}}
 {id=WavePlayer_339, properties {soundWaveAssetPtr=/Game/Sounds/Fortress_Foley/Commando/Commando_Land_Foley_Stereo_005}}
 {id=WavePlayer_341, properties {soundWaveAssetPtr=/Game/Sounds/Fortress_Foley/Commando/Commando_Land_Foley_Stereo_004}}
 {id=WavePlayer_4, properties {soundWaveAssetPtr=/Figment_Martin/Athena/Sounds/LTMs/Ashton/Milo/Ashton_Milo_Movment_Sweetener_02}}
 {id=WavePlayer_5, properties {soundWaveAssetPtr=/Figment_Martin/Athena/Sounds/LTMs/Ashton/Milo/Ashton_Milo_Movment_Sweetener_03}}
]
properties {
 attenuationSettings=/Game/Sounds/Attenuation/Fort_Player_Land_Athena_Att
 maxDistance=2097152
 soundClassObject=&Fort_Player_Foley_1P_Athena
 volumeMultiplier=1
 concurrencyOverrides {maxCount=4, resolutionRule=PreventNew}
 concurrencySet [/Game/Sounds/Concurrency_Rules/Athena_Footstep_Land_Conc]
}
