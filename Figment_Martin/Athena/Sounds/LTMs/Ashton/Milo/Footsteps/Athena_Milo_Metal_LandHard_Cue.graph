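# Sound cue for Milo's hard landings on metal in the Ashton LTM, using the first-person foley class, landing attenuation and capped concurrency.
# Mixes random soldier land and metal step waves under wide pitch and volume modulation, then layers in a separately modulated random Milo movement sweetener.
# A quality node gives its first two slots six-variant metal and sweetener pools, and its last two slots three-variant pools.
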SoundCue /Figment_Martin/Athena/Sounds/LTMs/Ashton/Milo/Footsteps/Athena_Milo_Metal_LandHard_Cue
root=QualityLevel_0
nodes [
 {id=Mixer_0, children [Random_1,Random_2], properties {inputVolume [0.6,0.9]}}
 {id=Mixer_2, children [Random_3,Random_0], properties {inputVolume [0.6,0.9]}}
 {id=Mixer_3, children [Modulator_4,Modulator_1], properties {inputVolume [1,1]}}
 {id=Mixer_4, children [Modulator_0,Modulator_2], properties {inputVolume [1,1]}}
 {id=Modulator_0, children [Mixer_0], properties {pitchMax=1.1, pitchMin=0.9, volumeMax=1.3, volumeMin=0.7}}
 {id=Modulator_1, children [Random_4]}
 {id=Modulator_2, children [Random_5]}
 {id=Modulator_4, children [Mixer_2], properties {pitchMax=1.1, pitchMin=0.9, volumeMax=1.3, volumeMin=0.7}}
 {id=QualityLevel_0, children [Mixer_3,Mixer_3,Mixer_4,Mixer_4]}
 {
  id=Random_0
  children [WavePlayer_34, WavePlayer_35, WavePlayer_36, WavePlayer_37, WavePlayer_38, WavePlayer_39]
  properties {bShouldExcludeFromBranchCulling=true, weights [1,1,1,1,1,1]}
 }
 {id=Random_1, children [WavePlayer_0,WavePlayer_1], properties {bShouldExcludeFromBranchCulling=true, weights [1,1]}}
 {id=Random_2, children [WavePlayer_2,WavePlayer_3,WavePlayer_4], properties {bShouldExcludeFromBranchCulling=true, weights [1,1,1]}}
 {id=Random_3, children [WavePlayer_19, WavePlayer_20], properties {bShouldExcludeFromBranchCulling=true, weights [1,1]}}
 {
  id=Random_4
  children [WavePlayer_29, WavePlayer_30, WavePlayer_31, WavePlayer_32, WavePlayer_33, WavePlayer_5]
  properties {weights [1,1,1,1,1,1]}
 }
 {id=Random_5, children [WavePlayer_6,WavePlayer_7,WavePlayer_8], properties {weights [1,1,1]}}
 {id=WavePlayer_0, properties {soundWaveAssetPtr=/Game/Sounds/Fort_Step_Sounds/Human/soldier_land_01}}
 {id=WavePlayer_1, properties {soundWaveAssetPtr=/Game/Sounds/Fort_Step_Sounds/Human/soldier_land_02}}
 {id=WavePlayer_19, properties {soundWaveAssetPtr=&soldier_land_01}}
 {id=WavePlayer_2, properties {soundWaveAssetPtr=/Game/Sounds/Fort_Step_Sounds/Soldier/Soldier_Step_Land_Metal_01}}
 {id=WavePlayer_20, properties {soundWaveAssetPtr=&soldier_land_02}}
 {id=WavePlayer_29, properties {soundWaveAssetPtr=/Figment_Martin/Athena/Sounds/LTMs/Ashton/Milo/Ashton_Milo_Movment_Sweetener_01}}
 {id=WavePlayer_3, properties {soundWaveAssetPtr=/Game/Sounds/Fort_Step_Sounds/Soldier/Soldier_Step_Land_Metal_02}}
 {id=WavePlayer_30, properties {soundWaveAssetPtr=/Figment_Martin/Athena/Sounds/LTMs/Ashton/Milo/Ashton_Milo_Movment_Sweetener_02}}
 {id=WavePlayer_31, properties {soundWaveAssetPtr=/Figment_Martin/Athena/Sounds/LTMs/Ashton/Milo/Ashton_Milo_Movment_Sweetener_03}}
 {id=WavePlayer_32, properties {soundWaveAssetPtr=/Figment_Martin/Athena/Sounds/LTMs/Ashton/Milo/Ashton_Milo_Movment_Sweetener_04}}
 {id=WavePlayer_33, properties {soundWaveAssetPtr=/Figment_Martin/Athena/Sounds/LTMs/Ashton/Milo/Ashton_Milo_Movment_Sweetener_05}}
 {id=WavePlayer_34, properties {soundWaveAssetPtr=&Soldier_Step_Land_Metal_01}}
 {id=WavePlayer_35, properties {soundWaveAssetPtr=&Soldier_Step_Land_Metal_02}}
 {id=WavePlayer_36, properties {soundWaveAssetPtr=/Game/Sounds/Fort_Step_Sounds/Soldier/Soldier_Step_Land_Metal_03}}
 {id=WavePlayer_37, properties {soundWaveAssetPtr=/Game/Sounds/Fort_Step_Sounds/Soldier/Soldier_Step_Land_Metal_04}}
 {id=WavePlayer_38, properties {soundWaveAssetPtr=/Game/Sounds/Fort_Step_Sounds/Soldier/Soldier_Step_Land_Metal_05}}
 {id=WavePlayer_39, properties {soundWaveAssetPtr=/Game/Sounds/Fort_Step_Sounds/Soldier/Soldier_Step_Land_Metal_06}}
 {id=WavePlayer_4, properties {soundWaveAssetPtr=&Soldier_Step_Land_Metal_03}}
 {id=WavePlayer_5, properties {soundWaveAssetPtr=/Figment_Martin/Athena/Sounds/LTMs/Ashton/Milo/Ashton_Milo_Movment_Sweetener_06}}
 {id=WavePlayer_6, properties {soundWaveAssetPtr=&Ashton_Milo_Movment_Sweetener_01}}
 {id=WavePlayer_7, properties {soundWaveAssetPtr=&Ashton_Milo_Movment_Sweetener_02}}
 {id=WavePlayer_8, properties {soundWaveAssetPtr=&Ashton_Milo_Movment_Sweetener_03}}
]
properties {
 attenuationSettings=/Game/Sounds/Attenuation/Fort_Player_Land_Athena_Att
 maxDistance=2097152
 soundClassObject=&Fort_Player_Foley_1P_Athena
 volumeMultiplier=1.2
 concurrencyOverrides {maxCount=4, resolutionRule=PreventNew}
 concurrencySet [/Game/Sounds/Concurrency_Rules/Athena_Footstep_Land_Conc]
}
