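# Sound cue for Milo's third-person sprint footstep on dirt, from the Ashton LTM sounds.
# A mixer layers a random Commando step, pitched up with varied volume, over a random Milo movement sweetener passed through a default modulator.
# The step randomizer is excluded from branch culling.
# Playback uses third-person sprint footstep attenuation, the friendly foley sound class, and local and global footstep concurrency sets.
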SoundCue /Figment_Martin/Athena/Sounds/LTMs/Ashton/Milo/Footsteps/Athena_Milo_Sprint_Dirt_Footstep_3P_Cue
root=Mixer_0
nodes [
 {id=Mixer_0, children [Modulator_2,Modulator_1], properties {inputVolume [1,1]}}
 {id=Modulator_1, children [Random_4]}
 {id=Modulator_2, children [Random_5], properties {pitchMax=1.5, pitchMin=1.3, volumeMax=1, volumeMin=0.7}}
 {
  id=Random_4
  children [WavePlayer_29, WavePlayer_30, WavePlayer_31, WavePlayer_32, WavePlayer_33, WavePlayer_34]
  properties {weights [1,1,1,1,1,1]}
 }
 {
  id=Random_5
  children [WavePlayer_18, WavePlayer_19, WavePlayer_20, WavePlayer_21, WavePlayer_22, WavePlayer_23]
  properties {bShouldExcludeFromBranchCulling=true, weights [1,1,1,1,1,1]}
 }
 {id=WavePlayer_18, properties {soundWaveAssetPtr=/Game/Sounds/Fortress_Foley/Commando/Commando_Step_001}}
 {id=WavePlayer_19, properties {soundWaveAssetPtr=/Game/Sounds/Fortress_Foley/Commando/Commando_Step_002}}
 {id=WavePlayer_20, properties {soundWaveAssetPtr=/Game/Sounds/Fortress_Foley/Commando/Commando_Step_003}}
 {id=WavePlayer_21, properties {soundWaveAssetPtr=/Game/Sounds/Fortress_Foley/Commando/Commando_Step_004}}
 {id=WavePlayer_22, properties {soundWaveAssetPtr=/Game/Sounds/Fortress_Foley/Commando/Commando_Step_006}}
 {id=WavePlayer_23, properties {soundWaveAssetPtr=/Game/Sounds/Fortress_Foley/Commando/Commando_Step_010}}
 {id=WavePlayer_29, properties {soundWaveAssetPtr=/Figment_Martin/Athena/Sounds/LTMs/Ashton/Milo/Ashton_Milo_Movment_Sweetener_01}}
 {id=WavePlayer_30, properties {soundWaveAssetPtr=/Figment_Martin/Athena/Sounds/LTMs/Ashton/Milo/Ashton_Milo_Movment_Sweetener_02}}
 {id=WavePlayer_31, properties {soundWaveAssetPtr=/Figment_Martin/Athena/Sounds/LTMs/Ashton/Milo/Ashton_Milo_Movment_Sweetener_03}}
 {id=WavePlayer_32, properties {soundWaveAssetPtr=/Figment_Martin/Athena/Sounds/LTMs/Ashton/Milo/Ashton_Milo_Movment_Sweetener_04}}
 {id=WavePlayer_33, properties {soundWaveAssetPtr=/Figment_Martin/Athena/Sounds/LTMs/Ashton/Milo/Ashton_Milo_Movment_Sweetener_05}}
 {id=WavePlayer_34, properties {soundWaveAssetPtr=/Figment_Martin/Athena/Sounds/LTMs/Ashton/Milo/Ashton_Milo_Movment_Sweetener_06}}
]
properties {
 attenuationSettings=/Game/Sounds/Attenuation/Fort_Player_Footsteps_3P_Sprint_Athena_Att
 maxDistance=3584
 soundClassObject=/Game/Sounds/FortSoundClassesAndModes/Fort_Player_Foley_3P_Friendly_Athena
 volumeMultiplier=2.7
 concurrencyOverrides {bLimitToOwner=true, maxCount=2, resolutionRule=StopOldest}
 concurrencySet [&Athena_Footstep_Local_Conc, &Athena_Footstep_Global_Conc]
}
