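# Sound cue for Milo's sprint footsteps on dirt in an Ashton LTM, using the local-player (1P) foley class and sprint attenuation.
# A Mixer layers two equally weighted Random picks: a Commando footstep and a Milo movement sweetener.
# A root Modulator randomizes volume and raises pitch well above normal for both layers.
# The cue joins the local footstep concurrency set, and its step Random is excluded from branch culling.
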SoundCue /Figment_Martin/Athena/Sounds/LTMs/Ashton/Milo/Footsteps/Athena_Milo_Sprint_Dirt_Footstep_Cue
root=Modulator_2
nodes [
 {id=Mixer_0, children [Random_5,Random_1], properties {inputVolume [1,1]}}
 {id=Modulator_2, children [Mixer_0], properties {pitchMax=1.5, pitchMin=1.3, volumeMax=1, volumeMin=0.7}}
 {
  id=Random_1
  children [WavePlayer_11, WavePlayer_12, WavePlayer_13, WavePlayer_14, WavePlayer_15, WavePlayer_16]
  properties {weights [1,1,1,1,1,1]}
 }
 {
  id=Random_5
  children [WavePlayer_18, WavePlayer_19, WavePlayer_20, WavePlayer_21, WavePlayer_22, WavePlayer_23]
  properties {bShouldExcludeFromBranchCulling=true, weights [1,1,1,1,1,1]}
 }
 {id=WavePlayer_11, properties {soundWaveAssetPtr=/Figment_Martin/Athena/Sounds/LTMs/Ashton/Milo/Ashton_Milo_Movment_Sweetener_01}}
 {id=WavePlayer_12, properties {soundWaveAssetPtr=/Figment_Martin/Athena/Sounds/LTMs/Ashton/Milo/Ashton_Milo_Movment_Sweetener_02}}
 {id=WavePlayer_13, properties {soundWaveAssetPtr=/Figment_Martin/Athena/Sounds/LTMs/Ashton/Milo/Ashton_Milo_Movment_Sweetener_03}}
 {id=WavePlayer_14, properties {soundWaveAssetPtr=/Figment_Martin/Athena/Sounds/LTMs/Ashton/Milo/Ashton_Milo_Movment_Sweetener_04}}
 {id=WavePlayer_15, properties {soundWaveAssetPtr=/Figment_Martin/Athena/Sounds/LTMs/Ashton/Milo/Ashton_Milo_Movment_Sweetener_05}}
 {id=WavePlayer_16, properties {soundWaveAssetPtr=/Figment_Martin/Athena/Sounds/LTMs/Ashton/Milo/Ashton_Milo_Movment_Sweetener_06}}
 {id=WavePlayer_18, properties {soundWaveAssetPtr=/Game/Sounds/Fortress_Foley/Commando/Commando_Step_001}}
 {id=WavePlayer_19, properties {soundWaveAssetPtr=/Game/Sounds/Fortress_Foley/Commando/Commando_Step_002}}
 {id=WavePlayer_20, properties {soundWaveAssetPtr=/Game/Sounds/Fortress_Foley/Commando/Commando_Step_003}}
 {id=WavePlayer_21, properties {soundWaveAssetPtr=/Game/Sounds/Fortress_Foley/Commando/Commando_Step_004}}
 {id=WavePlayer_22, properties {soundWaveAssetPtr=/Game/Sounds/Fortress_Foley/Commando/Commando_Step_006}}
 {id=WavePlayer_23, properties {soundWaveAssetPtr=/Game/Sounds/Fortress_Foley/Commando/Commando_Step_010}}
]
properties {
 attenuationSettings=/Game/Sounds/Attenuation/Fort_Player_Footsteps_1P_Sprint_Athena_Att
 maxDistance=2097152
 soundClassObject=&Fort_Player_Foley_1P_Athena
 concurrencyOverrides {bLimitToOwner=true, maxCount=1, resolutionRule=StopOldest}
 concurrencySet [&Athena_Footstep_Local_Conc]
}
