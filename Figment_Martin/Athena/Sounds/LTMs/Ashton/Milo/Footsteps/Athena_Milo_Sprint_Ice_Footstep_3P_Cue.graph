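# Sound cue for third-person sprint footsteps on ice for the Milo character in the Ashton LTM.
# A mixer layers a random soldier snow-run step, with randomized pitch and volume, over a random Milo movement sweetener through its own modulator.
# The ice step reuses snow recordings rather than dedicated ice samples.
# It uses sprint footstep attenuation, the friendly 3P foley sound class, and local and global footstep concurrency sets.
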SoundCue /Figment_Martin/Athena/Sounds/LTMs/Ashton/Milo/Footsteps/Athena_Milo_Sprint_Ice_Footstep_3P_Cue
root=Mixer_0
nodes [
 {id=Mixer_0, children [Modulator_2,Modulator_1], properties {inputVolume [1,1]}}
 {id=Modulator_1, children [Random_4]}
 {id=Modulator_2, children [Random_1], properties {pitchMax=1.07, pitchMin=1, volumeMax=1, volumeMin=0.7}}
 {
  id=Random_1
  children [
   WavePlayer_12
   WavePlayer_13
   WavePlayer_14
   WavePlayer_15
   WavePlayer_16
   WavePlayer_17
   WavePlayer_24
   WavePlayer_25
   WavePlayer_26
   WavePlayer_27
   WavePlayer_28
   WavePlayer_29
  ]
  properties {weights [1,1,1,1,1,1,1,1,1,1,1,1]}
 }
 {
  id=Random_4
  children [WavePlayer_0, WavePlayer_10, WavePlayer_31, WavePlayer_32, WavePlayer_33, WavePlayer_34]
  properties {weights [1,1,1,1,1,1]}
 }
 {id=WavePlayer_0, properties {soundWaveAssetPtr=/Figment_Martin/Athena/Sounds/LTMs/Ashton/Milo/Ashton_Milo_Movment_Sweetener_01}}
 {id=WavePlayer_10, properties {soundWaveAssetPtr=/Figment_Martin/Athena/Sounds/LTMs/Ashton/Milo/Ashton_Milo_Movment_Sweetener_02}}
 {id=WavePlayer_12, properties {soundWaveAssetPtr=/Game/Sounds/Fort_Step_Sounds/Soldier/Soldier_Step_Run_Snow_09}}
 {id=WavePlayer_13, properties {soundWaveAssetPtr=/Game/Sounds/Fort_Step_Sounds/Soldier/Soldier_Step_Run_Snow_10}}
 {id=WavePlayer_14, properties {soundWaveAssetPtr=/Game/Sounds/Fort_Step_Sounds/Soldier/Soldier_Step_Run_Snow_11}}
 {id=WavePlayer_15, properties {soundWaveAssetPtr=/Game/Sounds/Fort_Step_Sounds/Soldier/Soldier_Step_Run_Snow_12}}
 {id=WavePlayer_16, properties {soundWaveAssetPtr=/Game/Sounds/Fort_Step_Sounds/Soldier/Soldier_Step_Run_Snow_01}}
 {id=WavePlayer_17, properties {soundWaveAssetPtr=/Game/Sounds/Fort_Step_Sounds/Soldier/Soldier_Step_Run_Snow_02}}
 {id=WavePlayer_24, properties {soundWaveAssetPtr=/Game/Sounds/Fort_Step_Sounds/Soldier/Soldier_Step_Run_Snow_03}}
 {id=WavePlayer_25, properties {soundWaveAssetPtr=/Game/Sounds/Fort_Step_Sounds/Soldier/Soldier_Step_Run_Snow_04}}
 {id=WavePlayer_26, properties {soundWaveAssetPtr=/Game/Sounds/Fort_Step_Sounds/Soldier/Soldier_Step_Run_Snow_05}}
 {id=WavePlayer_27, properties {soundWaveAssetPtr=/Game/Sounds/Fort_Step_Sounds/Soldier/Soldier_Step_Run_Snow_06}}
 {id=WavePlayer_28, properties {soundWaveAssetPtr=/Game/Sounds/Fort_Step_Sounds/Soldier/Soldier_Step_Run_Snow_07}}
 {id=WavePlayer_29, properties {soundWaveAssetPtr=/Game/Sounds/Fort_Step_Sounds/Soldier/Soldier_Step_Run_Snow_08}}
 {id=WavePlayer_31, properties {soundWaveAssetPtr=/Figment_Martin/Athena/Sounds/LTMs/Ashton/Milo/Ashton_Milo_Movment_Sweetener_03}}
 {id=WavePlayer_32, properties {soundWaveAssetPtr=/Figment_Martin/Athena/Sounds/LTMs/Ashton/Milo/Ashton_Milo_Movment_Sweetener_04}}
 {id=WavePlayer_33, properties {soundWaveAssetPtr=/Figment_Martin/Athena/Sounds/LTMs/Ashton/Milo/Ashton_Milo_Movment_Sweetener_05}}
 {id=WavePlayer_34, properties {soundWaveAssetPtr=/Figment_Martin/Athena/Sounds/LTMs/Ashton/Milo/Ashton_Milo_Movment_Sweetener_06}}
]
properties {
 attenuationSettings=/Game/Sounds/Attenuation/Fort_Player_Footsteps_3P_Sprint_Athena_Att
 maxDistance=3584
 soundClassObject=/Game/Sounds/FortSoundClassesAndModes/Fort_Player_Foley_3P_Friendly_Athena
 volumeMultiplier=2.7
 concurrencyOverrides {bLimitToOwner=true, maxCount=1, resolutionRule=StopOldest}
 concurrencySet [&Athena_Footstep_Local_Conc, &Athena_Footstep_Global_Conc]
}
