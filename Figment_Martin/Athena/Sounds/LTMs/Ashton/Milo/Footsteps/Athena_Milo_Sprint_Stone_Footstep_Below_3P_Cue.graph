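# Sound cue for Milo's third-person sprint footsteps on stone, heard from below, using downstairs Soldier step waves.
# A quality node sends its first two levels to a mixer layering a random step with a louder random thump, and its last two to a smaller step-only pool.
# Each pool has random pitch and volume modulation.
# The cue uses sprint footstep attenuation, an above-or-below sound class and footstep concurrency.
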SoundCue /Figment_Martin/Athena/Sounds/LTMs/Ashton/Milo/Footsteps/Athena_Milo_Sprint_Stone_Footstep_Below_3P_Cue
root=QualityLevel_0
nodes [
 {id=Mixer_0, children [Modulator_2,Modulator_1], properties {inputVolume [1,1.6]}}
 {id=Modulator_0, children [Random_1], properties {pitchMax=1.07, pitchMin=1, volumeMax=1, volumeMin=0.7}}
 {id=Modulator_1, children [Random_4], properties {pitchMax=1.07, pitchMin=1, volumeMax=1, volumeMin=0.7}}
 {id=Modulator_2, children [Random_0], properties {pitchMax=1.07, pitchMin=1, volumeMax=1, volumeMin=0.7}}
 {id=QualityLevel_0, children [Mixer_0,Mixer_0,Modulator_0,Modulator_0]}
 {
  id=Random_0
  children [
   WavePlayer_1
   WavePlayer_2
   WavePlayer_3
   WavePlayer_4
   WavePlayer_5
   WavePlayer_6
   WavePlayer_7
   WavePlayer_8
   WavePlayer_9
   WavePlayer_10
   WavePlayer_11
   WavePlayer_18
   WavePlayer_19
  ]
  properties {weights [1,1,1,1,1,1,1,1,1,1,1,1,1]}
 }
 {
  id=Random_1
  children [WavePlayer_13, WavePlayer_14, WavePlayer_15, WavePlayer_24, WavePlayer_25, WavePlayer_26]
  properties {bShouldExcludeFromBranchCulling=true, weights [1,1,1,1,1,1]}
 }
 {
  id=Random_4
  children [WavePlayer_23, WavePlayer_20, WavePlayer_21, WavePlayer_22, WavePlayer_29, WavePlayer_30]
  properties {weights [1,1,1,1,1,1]}
 }
 {
  id=WavePlayer_1
  properties {soundWaveAssetPtr=/Game/Sounds/Fort_Step_Sounds/Soldier/Elevations/Downstairs/Soldier_Step_Walk_Stone_08_Downstairs}
 }
 {
  id=WavePlayer_10
  properties {soundWaveAssetPtr=/Game/Sounds/Fort_Step_Sounds/Soldier/Elevations/Downstairs/Soldier_Step_Run_Stone_05_Downstairs}
 }
 {
  id=WavePlayer_11
  properties {soundWaveAssetPtr=/Game/Sounds/Fort_Step_Sounds/Soldier/Elevations/Downstairs/Soldier_Step_Run_Stone_07_Downstairs}
 }
 {
  id=WavePlayer_13
  properties {soundWaveAssetPtr=/Game/Sounds/Fort_Step_Sounds/Soldier/Elevations/Downstairs/Soldier_Step_Run_Stone_02_Downstairs}
 }
 {
  id=WavePlayer_14
  properties {soundWaveAssetPtr=/Game/Sounds/Fort_Step_Sounds/Soldier/Elevations/Downstairs/Soldier_Step_Run_Stone_04_Downstairs}
 }
 {
  id=WavePlayer_15
  properties {soundWaveAssetPtr=/Game/Sounds/Fort_Step_Sounds/Soldier/Elevations/Downstairs/Soldier_Step_Run_Stone_06_Downstairs}
 }
 {
  id=WavePlayer_18
  properties {soundWaveAssetPtr=/Game/Sounds/Fort_Step_Sounds/Soldier/Elevations/Downstairs/Soldier_Step_Run_Stone_09_Downstairs}
 }
 {
  id=WavePlayer_19
  properties {soundWaveAssetPtr=/Game/Sounds/Fort_Step_Sounds/Soldier/Elevations/Downstairs/Soldier_Step_Run_Stone_11_Downstairs}
 }
 {
  id=WavePlayer_2
  properties {soundWaveAssetPtr=/Game/Sounds/Fort_Step_Sounds/Soldier/Elevations/Downstairs/Soldier_Step_Run_Stone_12_Downstairs}
 }
 {id=WavePlayer_20, properties {soundWaveAssetPtr=/Game/Sounds/Fort_Step_Sounds/Soldier/Elevations/Downstairs/FS_Downstairs_Stone_Thump_02}}
 {id=WavePlayer_21, properties {soundWaveAssetPtr=/Game/Sounds/Fort_Step_Sounds/Soldier/Elevations/Downstairs/FS_Downstairs_Stone_Thump_03}}
 {id=WavePlayer_22, properties {soundWaveAssetPtr=/Game/Sounds/Fort_Step_Sounds/Soldier/Elevations/Downstairs/FS_Downstairs_Stone_Thump_04}}
 {id=WavePlayer_23, properties {soundWaveAssetPtr=/Game/Sounds/Fort_Step_Sounds/Soldier/Elevations/Downstairs/FS_Downstairs_Stone_Thump_01}}
 {
  id=WavePlayer_24
  properties {soundWaveAssetPtr=/Game/Sounds/Fort_Step_Sounds/Soldier/Elevations/Downstairs/Soldier_Step_Run_Stone_01_Downstairs}
 }
 {
  id=WavePlayer_25
  properties {soundWaveAssetPtr=/Game/Sounds/Fort_Step_Sounds/Soldier/Elevations/Downstairs/Soldier_Step_Run_Stone_03_Downstairs}
 }
 {
  id=WavePlayer_26
  properties {soundWaveAssetPtr=&Soldier_Step_Run_Stone_05_Downstairs}
 }
 {id=WavePlayer_29, properties {soundWaveAssetPtr=/Game/Sounds/Fort_Step_Sounds/Soldier/Elevations/Downstairs/FS_Downstairs_Stone_Thump_05}}
 {
  id=WavePlayer_3
  properties {soundWaveAssetPtr=&Soldier_Step_Run_Stone_02_Downstairs}
 }
 {id=WavePlayer_30, properties {soundWaveAssetPtr=/Game/Sounds/Fort_Step_Sounds/Soldier/Elevations/Downstairs/FS_Downstairs_Stone_Thump_06}}
 {
  id=WavePlayer_4
  properties {soundWaveAssetPtr=&Soldier_Step_Run_Stone_04_Downstairs}
 }
 {
  id=WavePlayer_5
  properties {soundWaveAssetPtr=&Soldier_Step_Run_Stone_06_Downstairs}
 }
 {
  id=WavePlayer_6
  properties {soundWaveAssetPtr=/Game/Sounds/Fort_Step_Sounds/Soldier/Elevations/Downstairs/Soldier_Step_Run_Stone_08_Downstairs}
 }
 {
  id=WavePlayer_7
  properties {soundWaveAssetPtr=/Game/Sounds/Fort_Step_Sounds/Soldier/Elevations/Downstairs/Soldier_Step_Run_Stone_10_Downstairs}
 }
 {
  id=WavePlayer_8
  properties {soundWaveAssetPtr=&Soldier_Step_Run_Stone_01_Downstairs}
 }
 {
  id=WavePlayer_9
  properties {soundWaveAssetPtr=&Soldier_Step_Run_Stone_03_Downstairs}
 }
]
properties {
 attenuationSettings=/Game/Sounds/Attenuation/Fort_Player_Footsteps_3P_Sprint_Athena_Att
 maxDistance=3584
 soundClassObject=/Game/Sounds/FortSoundClassesAndModes/Fort_Player_Footsteps_AboveOrBelow
 volumeMultiplier=1.1
 concurrencyOverrides {bLimitToOwner=true, maxCount=1, resolutionRule=StopOldest}
 concurrencySet [&Athena_Footstep_Local_Conc, &Athena_Footstep_Global_Conc]
}
