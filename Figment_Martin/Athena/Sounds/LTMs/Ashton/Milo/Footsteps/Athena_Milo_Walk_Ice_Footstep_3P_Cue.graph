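# Sound cue for Milo's third-person walking footstep on ice, routed to the friendly 3P foley sound class with walk footstep attenuation.
# A Mixer layers one random Soldier walk snow step, standing in for ice, with one random Milo movement sweetener.
# The mix passes through a root Modulator.
# Local and global footstep concurrency sets are assigned, alongside override values for a single per-owner voice that stops the oldest.
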SoundCue /Figment_Martin/Athena/Sounds/LTMs/Ashton/Milo/Footsteps/Athena_Milo_Walk_Ice_Footstep_3P_Cue
root=Modulator_0
nodes [
 {id=Mixer_0, children [Random_0,Random_4], properties {inputVolume [1,1]}}
 {id=Modulator_0, children [Mixer_0]}
 {
  id=Random_0
  children [
   WavePlayer_0
   WavePlayer_1
   WavePlayer_2
   WavePlayer_3
   WavePlayer_4
   WavePlayer_5
   WavePlayer_12
   WavePlayer_13
   WavePlayer_14
   WavePlayer_15
   WavePlayer_16
   WavePlayer_17
  ]
  properties {weights [1,1,1,1,1,1,1,1,1,1,1,1]}
 }
 {
  id=Random_4
  children [WavePlayer_29, WavePlayer_30, WavePlayer_31, WavePlayer_32, WavePlayer_33, WavePlayer_34]
  properties {weights [1,1,1,1,1,1]}
 }
 {id=WavePlayer_0, properties {soundWaveAssetPtr=/Game/Sounds/Fort_Step_Sounds/Soldier/Soldier_Step_Walk_Snow_10}}
 {id=WavePlayer_1, properties {soundWaveAssetPtr=/Game/Sounds/Fort_Step_Sounds/Soldier/Soldier_Step_Walk_Snow_11}}
 {id=WavePlayer_12, properties {soundWaveAssetPtr=/Game/Sounds/Fort_Step_Sounds/Soldier/Soldier_Step_Walk_Snow_04}}
 {id=WavePlayer_13, properties {soundWaveAssetPtr=/Game/Sounds/Fort_Step_Sounds/Soldier/Soldier_Step_Walk_Snow_05}}
 {id=WavePlayer_14, properties {soundWaveAssetPtr=/Game/Sounds/Fort_Step_Sounds/Soldier/Soldier_Step_Walk_Snow_06}}
 {id=WavePlayer_15, properties {soundWaveAssetPtr=/Game/Sounds/Fort_Step_Sounds/Soldier/Soldier_Step_Walk_Snow_07}}
 {id=WavePlayer_16, properties {soundWaveAssetPtr=/Game/Sounds/Fort_Step_Sounds/Soldier/Soldier_Step_Walk_Snow_08}}
 {id=WavePlayer_17, properties {soundWaveAssetPtr=/Game/Sounds/Fort_Step_Sounds/Soldier/Soldier_Step_Walk_Snow_09}}
 {id=WavePlayer_2, properties {soundWaveAssetPtr=/Game/Sounds/Fort_Step_Sounds/Soldier/Soldier_Step_Walk_Snow_12}}
 {id=WavePlayer_29, properties {soundWaveAssetPtr=/Figment_Martin/Athena/Sounds/LTMs/Ashton/Milo/Ashton_Milo_Movment_Sweetener_01}}
 {id=WavePlayer_3, properties {soundWaveAssetPtr=/Game/Sounds/Fort_Step_Sounds/Soldier/Soldier_Step_Walk_Snow_01}}
 {id=WavePlayer_30, properties {soundWaveAssetPtr=/Figment_Martin/Athena/Sounds/LTMs/Ashton/Milo/Ashton_Milo_Movment_Sweetener_02}}
 {id=WavePlayer_31, properties {soundWaveAssetPtr=/Figment_Martin/Athena/Sounds/LTMs/Ashton/Milo/Ashton_Milo_Movment_Sweetener_03}}
 {id=WavePlayer_32, properties {soundWaveAssetPtr=/Figment_Martin/Athena/Sounds/LTMs/Ashton/Milo/Ashton_Milo_Movment_Sweetener_04}}
 {id=WavePlayer_33, properties {soundWaveAssetPtr=/Figment_Martin/Athena/Sounds/LTMs/Ashton/Milo/Ashton_Milo_Movment_Sweetener_05}}
 {id=WavePlayer_34, properties {soundWaveAssetPtr=/Figment_Martin/Athena/Sounds/LTMs/Ashton/Milo/Ashton_Milo_Movment_Sweetener_06}}
 {id=WavePlayer_4, properties {soundWaveAssetPtr=/Game/Sounds/Fort_Step_Sounds/Soldier/Soldier_Step_Walk_Snow_02}}
 {id=WavePlayer_5, properties {soundWaveAssetPtr=/Game/Sounds/Fort_Step_Sounds/Soldier/Soldier_Step_Walk_Snow_03}}
]
properties {
 attenuationSettings=/Game/Sounds/Attenuation/Fort_Player_Footsteps_3P_Walk_Athena_Att
 maxDistance=3584
 soundClassObject=/Game/Sounds/FortSoundClassesAndModes/Fort_Player_Foley_3P_Friendly_Athena
 volumeMultiplier=1.95
 concurrencyOverrides {bLimitToOwner=true, maxCount=1, resolutionRule=StopOldest}
 concurrencySet [&Athena_Footstep_Local_Conc, &Athena_Footstep_Global_Conc]
}
