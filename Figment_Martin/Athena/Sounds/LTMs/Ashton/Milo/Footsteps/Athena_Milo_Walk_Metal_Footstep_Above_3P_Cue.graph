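# Sound cue for Milo's third-person walk footsteps on metal heard from above in Athena, built from Soldier "Upstairs" step recordings.
# A quality-level node sends its first two slots to a random pick of ten waves and its last two to a three-wave subset.
# A fixed modulator boosts volume above that switch.
# Third-person walk attenuation, an above-or-below footstep sound class and local and global concurrency sets govern playback.
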SoundCue /Figment_Martin/Athena/Sounds/LTMs/Ashton/Milo/Footsteps/Athena_Milo_Walk_Metal_Footstep_Above_3P_Cue
root=Modulator_4
nodes [
 {id=Modulator_4, children [QualityLevel_1], properties {volumeMax=1.5, volumeMin=1.5}}
 {id=QualityLevel_1, children [Random_2,Random_2,Random_3,Random_3]}
 {
  id=Random_2
  children [
   WavePlayer_22
   WavePlayer_23
   WavePlayer_24
   WavePlayer_25
   WavePlayer_26
   WavePlayer_27
   WavePlayer_28
   WavePlayer_29
   WavePlayer_30
   WavePlayer_31
  ]
  properties {weights [1,1,1,1,1,1,1,1,1,1]}
 }
 {id=Random_3, children [WavePlayer_32, WavePlayer_33, WavePlayer_46], properties {weights [1,1,1]}}
 {
  id=WavePlayer_22
  properties {soundWaveAssetPtr=/Game/Sounds/Fort_Step_Sounds/Soldier/Elevations/Upstairs/Soldier_Step_Walk_Metal_04_Upstairs2}
 }
 {
  id=WavePlayer_23
  properties {soundWaveAssetPtr=/Game/Sounds/Fort_Step_Sounds/Soldier/Elevations/Upstairs/Soldier_Step_Walk_Metal_11_Upstairs2}
 }
 {
  id=WavePlayer_24
  properties {soundWaveAssetPtr=/Game/Sounds/Fort_Step_Sounds/Soldier/Elevations/Upstairs/Soldier_Step_Walk_Metal_01_Upstairs2}
 }
 {
  id=WavePlayer_25
  properties {soundWaveAssetPtr=/Game/Sounds/Fort_Step_Sounds/Soldier/Elevations/Upstairs/Soldier_Step_Walk_Metal_02_Upstairs2}
 }
 {
  id=WavePlayer_26
  properties {soundWaveAssetPtr=/Game/Sounds/Fort_Step_Sounds/Soldier/Elevations/Upstairs/Soldier_Step_Walk_Metal_03_Upstairs2}
 }
 {
  id=WavePlayer_27
  properties {soundWaveAssetPtr=/Game/Sounds/Fort_Step_Sounds/Soldier/Elevations/Upstairs/Soldier_Step_Walk_Metal_05_Upstairs2}
 }
 {
  id=WavePlayer_28
  properties {soundWaveAssetPtr=/Game/Sounds/Fort_Step_Sounds/Soldier/Elevations/Upstairs/Soldier_Step_Walk_Metal_06_Upstairs2}
 }
 {
  id=WavePlayer_29
  properties {soundWaveAssetPtr=/Game/Sounds/Fort_Step_Sounds/Soldier/Elevations/Upstairs/Soldier_Step_Walk_Metal_07_Upstairs2}
 }
 {
  id=WavePlayer_30
  properties {soundWaveAssetPtr=/Game/Sounds/Fort_Step_Sounds/Soldier/Elevations/Upstairs/Soldier_Step_Walk_Metal_10_Upstairs2}
 }
 {
  id=WavePlayer_31
  properties {soundWaveAssetPtr=/Game/Sounds/Fort_Step_Sounds/Soldier/Elevations/Upstairs/Soldier_Step_Walk_Metal_12_Upstairs2}
 }
 {
  id=WavePlayer_32
  properties {soundWaveAssetPtr=&Soldier_Step_Walk_Metal_04_Upstairs2}
 }
 {
  id=WavePlayer_33
  properties {soundWaveAssetPtr=&Soldier_Step_Walk_Metal_11_Upstairs2}
 }
 {
  id=WavePlayer_46
  properties {soundWaveAssetPtr=&Soldier_Step_Walk_Metal_01_Upstairs2}
 }
]
properties {
 attenuationSettings=/Game/Sounds/Attenuation/Fort_Player_Footsteps_3P_Walk_Athena_Att
 maxDistance=3584
 soundClassObject=/Game/Sounds/FortSoundClassesAndModes/Fort_Player_Footsteps_AboveOrBelow
 volumeMultiplier=0.65
 concurrencyOverrides {bLimitToOwner=true, maxCount=1, resolutionRule=StopOldest}
 concurrencySet [&Athena_Footstep_Local_Conc, &Athena_Footstep_Global_Conc]
}
