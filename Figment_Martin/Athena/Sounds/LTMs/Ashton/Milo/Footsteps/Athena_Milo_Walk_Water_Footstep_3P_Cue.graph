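# Third-person walk footstep cue for the Ashton Milo character on water, played with walk footstep attenuation and friendly-player foley sound class.
# Each variant mixes a random Soldier run-water step with a random Milo movement sweetener, then passes it through a modulator.
# The quality switch sends its first two slots to the larger wave pools and its last two to reduced pools.
# Per-owner concurrency overrides stop the oldest sound.
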SoundCue /Figment_Martin/Athena/Sounds/LTMs/Ashton/Milo/Footsteps/Athena_Milo_Walk_Water_Footstep_3P_Cue
root=QualityLevel_0
nodes [
 {id=Mixer_0, children [Random_2,Random_4], properties {inputVolume [1,1]}}
 {id=Mixer_1, children [Random_0,Random_1], properties {inputVolume [1,1]}}
 {id=Modulator_0, children [Mixer_0]}
 {id=Modulator_1, children [Mixer_1]}
 {id=QualityLevel_0, children [Modulator_0,Modulator_0,Modulator_1,Modulator_1]}
 {id=Random_0, children [WavePlayer_8,WavePlayer_0,WavePlayer_6,WavePlayer_7], properties {weights [1,1,1,1]}}
 {id=Random_1, children [WavePlayer_11, WavePlayer_12, WavePlayer_13], properties {weights [1,1,1]}}
 {
  id=Random_2
  children [
   WavePlayer_1
   WavePlayer_2
   WavePlayer_3
   WavePlayer_4
   WavePlayer_5
   WavePlayer_24
   WavePlayer_25
   WavePlayer_26
   WavePlayer_27
   WavePlayer_28
   WavePlayer_29
   WavePlayer_30
  ]
  properties {bShouldExcludeFromBranchCulling=true, weights [1,1,1,1,1,1,1,1,1,1,1,1]}
 }
 {
  id=Random_4
  children [WavePlayer_9, WavePlayer_10, WavePlayer_31, WavePlayer_32, WavePlayer_33, WavePlayer_34]
  properties {weights [1,1,1,1,1,1]}
 }
 {id=WavePlayer_0, properties {soundWaveAssetPtr=/Game/Sounds/Fort_Step_Sounds/Soldier/Soldier_Step_Run_Water_10}}
 {id=WavePlayer_1, properties {soundWaveAssetPtr=&Soldier_Step_Run_Water_10}}
 {id=WavePlayer_10, properties {soundWaveAssetPtr=/Figment_Martin/Athena/Sounds/LTMs/Ashton/Milo/Ashton_Milo_Movment_Sweetener_02}}
 {id=WavePlayer_11, properties {soundWaveAssetPtr=/Figment_Martin/Athena/Sounds/LTMs/Ashton/Milo/Ashton_Milo_Movment_Sweetener_01}}
 {id=WavePlayer_12, properties {soundWaveAssetPtr=&Ashton_Milo_Movment_Sweetener_02}}
 {id=WavePlayer_13, properties {soundWaveAssetPtr=/Figment_Martin/Athena/Sounds/LTMs/Ashton/Milo/Ashton_Milo_Movment_Sweetener_03}}
 {id=WavePlayer_2, properties {soundWaveAssetPtr=/Game/Sounds/Fort_Step_Sounds/Soldier/Soldier_Step_Run_Water_11}}
 {id=WavePlayer_24, properties {soundWaveAssetPtr=/Game/Sounds/Fort_Step_Sounds/Soldier/Soldier_Step_Run_Water_03}}
 {id=WavePlayer_25, properties {soundWaveAssetPtr=/Game/Sounds/Fort_Step_Sounds/Soldier/Soldier_Step_Run_Water_04}}
 {id=WavePlayer_26, properties {soundWaveAssetPtr=/Game/Sounds/Fort_Step_Sounds/Soldier/Soldier_Step_Run_Water_05}}
 {id=WavePlayer_27, properties {soundWaveAssetPtr=/Game/Sounds/Fort_Step_Sounds/Soldier/Soldier_Step_Run_Water_06}}
 {id=WavePlayer_28, properties {soundWaveAssetPtr=/Game/Sounds/Fort_Step_Sounds/Soldier/Soldier_Step_Run_Water_07}}
 {id=WavePlayer_29, properties {soundWaveAssetPtr=/Game/Sounds/Fort_Step_Sounds/Soldier/Soldier_Step_Run_Water_08}}
 {id=WavePlayer_3, properties {soundWaveAssetPtr=/Game/Sounds/Fort_Step_Sounds/Soldier/Soldier_Step_Run_Water_12}}
 {id=WavePlayer_30, properties {soundWaveAssetPtr=/Game/Sounds/Fort_Step_Sounds/Soldier/Soldier_Step_Run_Water_09}}
 {id=WavePlayer_31, properties {soundWaveAssetPtr=&Ashton_Milo_Movment_Sweetener_03}}
 {id=WavePlayer_32, properties {soundWaveAssetPtr=/Figment_Martin/Athena/Sounds/LTMs/Ashton/Milo/Ashton_Milo_Movment_Sweetener_04}}
 {id=WavePlayer_33, properties {soundWaveAssetPtr=/Figment_Martin/Athena/Sounds/LTMs/Ashton/Milo/Ashton_Milo_Movment_Sweetener_05}}
 {id=WavePlayer_34, properties {soundWaveAssetPtr=/Figment_Martin/Athena/Sounds/LTMs/Ashton/Milo/Ashton_Milo_Movment_Sweetener_06}}
 {id=WavePlayer_4, properties {soundWaveAssetPtr=/Game/Sounds/Fort_Step_Sounds/Soldier/Soldier_Step_Run_Water_01}}
 {id=WavePlayer_5, properties {soundWaveAssetPtr=/Game/Sounds/Fort_Step_Sounds/Soldier/Soldier_Step_Run_Water_02}}
 {id=WavePlayer_6, properties {soundWaveAssetPtr=&Soldier_Step_Run_Water_11}}
 {id=WavePlayer_7, properties {soundWaveAssetPtr=&Soldier_Step_Run_Water_12}}
 {id=WavePlayer_8, properties {soundWaveAssetPtr=&Soldier_Step_Run_Water_09}}
 {id=WavePlayer_9, properties {soundWaveAssetPtr=&Ashton_Milo_Movment_Sweetener_01}}
]
properties {
 attenuationSettings=/Game/Sounds/Attenuation/Fort_Player_Footsteps_3P_Walk_Athena_Att
 maxDistance=3584
 soundClassObject=/Game/Sounds/FortSoundClassesAndModes/Fort_Player_Foley_3P_Friendly_Athena
 volumeMultiplier=1.95
 concurrencyOverrides {bLimitToOwner=true, maxCount=2, resolutionRule=StopOldest}
 concurrencySet [&Athena_Footstep_Local_Conc, &Athena_Footstep_Global_Conc]
}
