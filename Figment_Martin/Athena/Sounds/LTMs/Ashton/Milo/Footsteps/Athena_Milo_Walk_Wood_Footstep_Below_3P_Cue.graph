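# Third-person sound cue for Milo's walking footsteps on wood heard from below, using shared Soldier downstairs samples and the above-or-below sound class.
# The first two quality levels mix a random step with a boosted random thump, each through a modulator.
# The last two play only one modulated step from a smaller pool, which is excluded from branch culling.
# Walk attenuation and footstep concurrency sets govern playback.
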SoundCue /Figment_Martin/Athena/Sounds/LTMs/Ashton/Milo/Footsteps/Athena_Milo_Walk_Wood_Footstep_Below_3P_Cue
root=QualityLevel_0
nodes [
 {id=Mixer_0, children [Modulator_1,Modulator_3], properties {inputVolume [1,1.6]}}
 {id=Modulator_0, children [Random_1]}
 {id=Modulator_1, children [Random_0]}
 {id=Modulator_3, children [Random_4]}
 {id=QualityLevel_0, children [Mixer_0,Mixer_0,Modulator_0,Modulator_0]}
 {
  id=Random_0
  children [
   WavePlayer_3
   WavePlayer_8
   WavePlayer_9
   WavePlayer_11
   WavePlayer_12
   WavePlayer_13
   WavePlayer_14
   WavePlayer_15
   WavePlayer_16
   WavePlayer_17
   WavePlayer_0
   WavePlayer_1
   WavePlayer_2
  ]
  properties {weights [1,1,1,1,1,1,1,1,1,1,1,1,1]}
 }
 {
  id=Random_1
  children [WavePlayer_4, WavePlayer_5, WavePlayer_6, WavePlayer_7, WavePlayer_10, WavePlayer_21]
  properties {bShouldExcludeFromBranchCulling=true, weights [1,1,1,1,1,1]}
 }
 {id=Random_4, children [WavePlayer_18, WavePlayer_19, WavePlayer_20, WavePlayer_24, WavePlayer_25], properties {weights [1,1,1,1,1]}}
 {
  id=WavePlayer_0
  properties {soundWaveAssetPtr=/Game/Sounds/Fort_Step_Sounds/Soldier/Elevations/Downstairs/Soldier_Step_Walk_Wood_10_Downstairs}
 }
 {
  id=WavePlayer_1
  properties {soundWaveAssetPtr=/Game/Sounds/Fort_Step_Sounds/Soldier/Elevations/Downstairs/Soldier_Step_Walk_Wood_11_Downstairs}
 }
 {
  id=WavePlayer_10
  properties {soundWaveAssetPtr=/Game/Sounds/Fort_Step_Sounds/Soldier/Elevations/Downstairs/Soldier_Step_Walk_Wood_05_Downstairs}
 }
 {
  id=WavePlayer_11
  properties {soundWaveAssetPtr=/Game/Sounds/Fort_Step_Sounds/Soldier/Elevations/Downstairs/Soldier_Step_Walk_Wood_03_Downstairs}
 }
 {
  id=WavePlayer_12
  properties {soundWaveAssetPtr=/Game/Sounds/Fort_Step_Sounds/Soldier/Elevations/Downstairs/Soldier_Step_Walk_Wood_04_Downstairs}
 }
 {
  id=WavePlayer_13
  properties {soundWaveAssetPtr=&Soldier_Step_Walk_Wood_05_Downstairs}
 }
 {
  id=WavePlayer_14
  properties {soundWaveAssetPtr=/Game/Sounds/Fort_Step_Sounds/Soldier/Elevations/Downstairs/Soldier_Step_Walk_Wood_06_Downstairs}
 }
 {
  id=WavePlayer_15
  properties {soundWaveAssetPtr=/Game/Sounds/Fort_Step_Sounds/Soldier/Elevations/Downstairs/Soldier_Step_Walk_Wood_07_Downstairs}
 }
 {
  id=WavePlayer_16
  properties {soundWaveAssetPtr=/Game/Sounds/Fort_Step_Sounds/Soldier/Elevations/Downstairs/Soldier_Step_Walk_Wood_08_Downstairs}
 }
 {
  id=WavePlayer_17
  properties {soundWaveAssetPtr=/Game/Sounds/Fort_Step_Sounds/Soldier/Elevations/Downstairs/Soldier_Step_Walk_Wood_09_Downstairs}
 }
 {id=WavePlayer_18, properties {soundWaveAssetPtr=/Game/Sounds/Fort_Step_Sounds/Soldier/Elevations/Downstairs/FS_Downstairs_Wood_Thump_17}}
 {id=WavePlayer_19, properties {soundWaveAssetPtr=/Game/Sounds/Fort_Step_Sounds/Soldier/Elevations/Downstairs/FS_Downstairs_Wood_Thump_13}}
 {
  id=WavePlayer_2
  properties {soundWaveAssetPtr=/Game/Sounds/Fort_Step_Sounds/Soldier/Elevations/Downstairs/Soldier_Step_Walk_Wood_12_Downstairs}
 }
 {id=WavePlayer_20, properties {soundWaveAssetPtr=/Game/Sounds/Fort_Step_Sounds/Soldier/Elevations/Downstairs/FS_Downstairs_Wood_Thump_14}}
 {
  id=WavePlayer_21
  properties {soundWaveAssetPtr=&Soldier_Step_Walk_Wood_06_Downstairs}
 }
 {id=WavePlayer_24, properties {soundWaveAssetPtr=/Game/Sounds/Fort_Step_Sounds/Soldier/Elevations/Downstairs/FS_Downstairs_Wood_Thump_15}}
 {id=WavePlayer_25, properties {soundWaveAssetPtr=/Game/Sounds/Fort_Step_Sounds/Soldier/Elevations/Downstairs/FS_Downstairs_Wood_Thump_16}}
 {
  id=WavePlayer_3
  properties {soundWaveAssetPtr=&Soldier_Step_Walk_Wood_07_Downstairs}
 }
 {
  id=WavePlayer_4
  properties {soundWaveAssetPtr=/Game/Sounds/Fort_Step_Sounds/Soldier/Elevations/Downstairs/Soldier_Step_Walk_Wood_01_Downstairs}
 }
 {
  id=WavePlayer_5
  properties {soundWaveAssetPtr=/Game/Sounds/Fort_Step_Sounds/Soldier/Elevations/Downstairs/Soldier_Step_Walk_Wood_02_Downstairs}
 }
 {
  id=WavePlayer_6
  properties {soundWaveAssetPtr=&Soldier_Step_Walk_Wood_03_Downstairs}
 }
 {
  id=WavePlayer_7
  properties {soundWaveAssetPtr=&Soldier_Step_Walk_Wood_04_Downstairs}
 }
 {
  id=WavePlayer_8
  properties {soundWaveAssetPtr=&Soldier_Step_Walk_Wood_01_Downstairs}
 }
 {
  id=WavePlayer_9
  properties {soundWaveAssetPtr=&Soldier_Step_Walk_Wood_02_Downstairs}
 }
]
properties {
 attenuationSettings=/Game/Sounds/Attenuation/Fort_Player_Footsteps_3P_Walk_Athena_Att
 maxDistance=3584
 soundClassObject=/Game/Sounds/FortSoundClassesAndModes/Fort_Player_Footsteps_AboveOrBelow
 volumeMultiplier=0.85
 concurrencyOverrides {bLimitToOwner=true, maxCount=1, resolutionRule=StopOldest}
 concurrencySet [&Athena_Footstep_Local_Conc, &Athena_Footstep_Global_Conc]
}
